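; NPC pawn for the Spire boss, which only takes damage routed through its own weakpoint actors.
; Incoming damage is zeroed unless the weakpoints utility component recognizes the damage causer as an owned weakpoint.
; Max health scales with the playlist's max team size once playlist data is ready.
; Damage effects spawn a ricochet impact along the hit normal, and server death broadcasts a boss-died event through the gameplay event router as a temporary measure.
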(event EventBeginPlay
  (|CallParentFunction)
  ;[0 0.6719 0.1154 0.5] Set Scaled Health Value
  (bind (_gamestate _playlist) (Class|FortAsyncActionCurrentPlaylistReady|PlaylistDataReadyAsync)
    (:OnReady
      (FortPawn|SetMaxHealth (* (ScalableFloat|GetValueatLevel (Variables|InternalData|GetSF_DefaultHealthScalar) (Math|Conversions|ToFloat(Integer) (Variables|/Script/FortniteGame.FortPlaylist|GetMaxTeamSize _playlist))) 1e+02)))))

(event EventOnDamagePlayEffects
  (params (Damage :type Float%20%28single-precision%29) (DamageTags :type Gameplay%20Tag%20Container%20Structure%20%28by%20ref%29) (Momentum :type Vector :default "0, 0, 0") (HitInfo :type Hit%20Result%20Structure%20%28by%20ref%29) (InstigatedBy :type Fort%20Pawn%20Object%20Reference) (DamageCauser :type Actor%20Object%20Reference) (EffectContext :type Gameplay%20Effect%20Context%20Handle%20Structure))
  (bind (_bblockinghit _binitialoverlap _time _distance _location _impactpoint _normal) (Collision|BreakHitResult HitInfo))
  ;[0.9271 0.1015 0.09266 0.5] Play Ricochet Hit VFX
  (bind _returnvalue (Effects|Components|ParticleSystem|SpawnFXAtLocation "/SpireSharedAssets/VFX/NS_SpireBoss_RicochetImpact.NS_SpireBoss_RicochetImpact" _location (Math|Conversions|RotationFromXVector _normal)))
  (Effects|Components|ParticleSystem|SetVectorParameter _returnvalue "ImpactFXDirection" _normal))

(event EventOnDeathServer
  (params (Damage :type Float%20%28single-precision%29) (DamageTags :type Gameplay%20Tag%20Container%20Structure%20%28by%20ref%29) (Momentum :type Vector :default "0, 0, 0") (HitInfo :type Hit%20Result%20Structure%20%28by%20ref%29) (InstigatedBy :type Controller%20Object%20Reference) (DamageCauser :type Actor%20Object%20Reference) (EffectContext :type Gameplay%20Effect%20Context%20Handle%20Structure))
  ;[0.15 0.1217 0 0.5] Adding temp death broadcasting. Will change once we move to BGA
  (bind _returnvalue (GameplayEventRouter|GetEventRouter "/Script/FortniteGame.FortPlayspace" self))
  (GameplayEventRouter|BroadcastEvent _returnvalue (Utilities|Struct|MakeOnSpireBossDied InstigatedBy) 0))

(event Collision|EventActorBeginOverlap
  (params (OtherActor :type Actor%20Object%20Reference))
  (|CallParentFunction OtherActor) ; This node is disabled and will not be called.\nDrag off pins to build functionality.
)

(event EventTick
  (params (DeltaSeconds :type Float%20%28single-precision%29))
  (|CallParentFunction DeltaSeconds) ; This node is disabled and will not be called.\nDrag off pins to build functionality.
)

(fn BlueprintModifyIncomingDamage
  (params (Damage :type Float%20%28single-precision%29) (DamageTags :type Gameplay%20Tag%20Container%20Structure%20%28by%20ref%29) (EffectContext :type Gameplay%20Effect%20Context%20Handle%20Structure%20%28by%20ref%29) (EventInstigator :type Controller%20Object%20Reference) (DamageCauser :type Actor%20Object%20Reference))
  (bind _returnvalue (|CallParentFunction Damage DamageTags EffectContext EventInstigator DamageCauser))
  (Development|PrintString "@@@ BlueprintModifyIncomingDamage [BP_NPCPawn_SpireBoss]: CALLED!" false (Variables|Debug|GetbDebugSpire))
  (bind _bisweakpoint (Class|BPActorWeakpointsUtilityComponent|IsOwnedWeakpointActor (Variables|Default|GetBP_ActorWeakpoints_UtilityComponent) DamageCauser true))
  (if _bisweakpoint
    (Development|PrintString (Utilities|String|Append "@@@@@@ BlueprintModifyIncomingDamage [BP_NPCPawn_SpireBoss]: WEAKPOINT DAMAGE = " (Utilities|String|ToString(Float) _returnvalue) "!") false (Variables|Debug|GetbDebugSpire))
    (return _returnvalue)
    (else
      (return 0.0))))

(fn ConstructionScript
  (|CallParentFunction))
